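; Example fluid simulation Blueprint that drives a moving dynamic force from an oscillating mesh.
; After a short delay on BeginPlay, it registers the force mesh with the referenced BPFluidSim01, passing radius, strength and material.
; Tick moves the mesh's relative location along X with a game-time sine wave scaled by speed and magnitude.
; The construction script scales the mesh to the radius and toggles its in-game visibility.
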(event EventTick
  (params (DeltaSeconds :type Float%20%28single-precision%29))
  (Transformation|SetRelativeLocation (Variables|Default|GetForceMesh) (Math|Vector|MakeVector (* (Math|Trig|Sin(Radians) (* (Utilities|Time|GetGameTimeinSeconds) (Variables|Default|GetOscillation%20Speed))) (Variables|Default|GetOsc%20Mag))) false false))

(event EventBeginPlay
  (Utilities|FlowControl|Delay 0.1)
  (Class|BPFluidSim01|RegisterDynamicForce (Variables|Default|GetSim) (Utilities|Struct|MakeFluidForceDynamic "NewEnumerator0" (Variables|Default|GetRadius) (Variables|Default|GetStrength) (Variables|Default|GetForceMesh) (Variables|Default|GetMaterial) (Utilities|Struct|MakeFluidForceSocketInfo (Utilities|Map|MakeMap "None" "None"))) (Variables|Default|GetForceMesh) 0.0))

(fn ConstructionScript
  (Transformation|SetWorldScale3D (Variables|Default|GetForceMesh) (Math|Conversions|ToVector(Float) (* (Variables|Default|GetRadius) 1)))
  (Development|SetHiddeninGame (Variables|Default|GetForceMesh) (not (Variables|Default|GetShow%20Mesh)) false))
