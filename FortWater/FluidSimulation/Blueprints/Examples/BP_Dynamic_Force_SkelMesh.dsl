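; Example Blueprint that registers a skeletal mesh as a dynamic force in a BP_FluidSim01 fluid simulation.
; On BeginPlay it waits briefly, then sets members in its FluidForceDynamic settings struct using the skeletal mesh.
; It then calls RegisterDynamicForce on the referenced sim with those settings and the mesh.
; Tick and the construction script are empty, so the actor itself does no per-frame work.
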(event EventTick
  (params (DeltaSeconds :type Float%20%28single-precision%29)))

(event EventBeginPlay
  (Utilities|FlowControl|Delay 0.1)
  (Utilities|Struct|SetmembersinFluidForceDynamic (Variables|Default|GetFluidForceDynamic%20Settings) (Variables|Default|GetSkeletalMesh))
  (Class|BPFluidSim01|RegisterDynamicForce (Variables|Default|GetSim) (Variables|Default|GetFluidForceDynamic%20Settings) (Variables|Default|GetSkeletalMesh) 0.0))

(fn ConstructionScript)
