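; Example water plane actor that feeds the fluid simulation's render target into a material on its static mesh component.
; The construction script applies a configurable mesh and material to the plane when the mesh is valid.
; On BeginPlay, after a short delay, it creates and stores a dynamic material instance, then passes the render target from GetFluidSimRenderTarget to its NormalAndHeight texture parameter.
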(event EventBeginPlay
  (Utilities|FlowControl|Delay 3.0)
  (bind _returnvalue (Rendering|Material|CreateDynamicMaterialInstance (Variables|Default|GetS_WaterPlane_512_LOD0) 0 (Variables|Default|GetMaterial%20to%20Use) "None" :!class /Script/Engine.PrimitiveComponent))
  (Variables|Default|SetMID _returnvalue)
  (bind _normal_and_height_rt (Utilities|GetFluidSimRenderTarget)
    (:Out
      (Rendering|Material|SetTextureParameterValue (Variables|Default|GetMID) "NormalAndHeight" _normal_and_height_rt))))

(fn ConstructionScript
  (bind _static_mesh (Variables|Default|GetStatic%20Mesh))
  (Utilities|IsValid _static_mesh
    (:Is%20Valid
      (Components|StaticMesh|SetStaticMesh (Variables|Default|GetS_WaterPlane_512_LOD0) _static_mesh)
      (Rendering|Material|SetMaterial (Variables|Default|GetS_WaterPlane_512_LOD0) 0 (Variables|Default|GetMaterial%20to%20Use)))))
