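; BP_WaterSim_TickInEditorTest is a small test Blueprint among the fluid simulation examples that, going by its name, checks actor ticking in the editor.
; The construction script disables actor tick, and the custom events Start Tick and Stop Tick turn it on and off.
; While ticking, it prints "Hello" every tick; BeginPlay and ActorBeginOverlap are empty.
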(event EventBeginPlay)

(event Collision|EventActorBeginOverlap
  (params (OtherActor :type Actor%20Object%20Reference)))

(event EventTick
  (params (DeltaSeconds :type Float%20%28single-precision%29))
  (Development|PrintString "Hello"))

(event Custom|Start%20Tick
  (Actor|Tick|SetActorTickEnabled true))

(event Custom|Stop%20Tick
  (Actor|Tick|SetActorTickEnabled false))

(fn ConstructionScript
  (Actor|Tick|SetActorTickEnabled false))
